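# Masked, lit mesh-particle material for the Sleepy Time pickaxe star effect, giving the star a bright glowing edge.
# A Fresnel term tints a gold or custom edge color, which is heavily boosted, oversaturated through negative desaturation and passed into MF_FX_EmissiveAdjustments.
# Scalar parameters blend in a custom edge color, a custom Fresnel exponent, and particle color and alpha.
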Material /BRCosmetics/Weapons/FORT_Melee/Pickaxe_Sleepy_Time/FX/M_Pickaxe_SleepyTime_Star
graph {
 properties {
  bUsedWithMeshParticles=true
  bUsedWithNiagaraMeshParticles=true
  blendMode=BLEND_Masked
  pixelDepthOffsetMode=PDOM_Legacy
  previewMesh=/Game/Weapons/FORT_Melee/Pickaxe_Sleepy_Time/FX/SM_Pickaxe_SleepyTime_Star
  usedShadingModels=Default Lit
 }
 nodes [
  Constant()
  Constant() {r=1}
  n1 = ParticleColor()
  n2 = ScalarParameter() {parameterName=Use Particle Color}
  LinearInterpolate(A=Constant() {r=1}, B=@n1.A, Alpha=@n2)
  n3 = LinearInterpolate(
   A=Constant3Vector() {constant=<rgba 1 0.7485 0.0208 1>}
   B=VectorParameter() {parameterName=Edge Color, defaultValue=<rgba 1 0.7485 0.0208 1>}.RGB
   Alpha=ScalarParameter() {parameterName=Use Custom Edge Color}
  )
  n4 = Fresnel(ExponentIn=LinearInterpolate(A=Constant() {r=1.4}, B=ScalarParameter() {parameterName=Fresnel Exp}, Alpha=ScalarParameter() {parameterName=Use Custom Fresnel Exp})) {baseReflectFraction=0.01, exponent=1.4}
  n5 = Desaturation(None=Multiply(A=Multiply(A=@n3, B=@n4), B=Constant() {r=20}), Fraction=Constant() {r=-1})
  MaterialFunctionCall(Emissive=LinearInterpolate(A=@n5, B=Multiply(A=@n5, B=@n1.RGB), Alpha=@n2)) {materialFunction=&MF_FX_EmissiveAdjustments}
  VectorParameter() {defaultValue=<rgba 0.5 0.1844 0 1>}
 ]
}
