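# Translucent, unlit, two-sided beam-trail material for the Sleepy Time pickaxe, reusing the character's body mask textures.
# Emissive layers panning, rounded mask shapes tinted cyan to yellow over a noise-modulated pink base, boosted in saturation by negative desaturation.
# Opacity is eroded by a Ceil-quantized "Beat Texture" and panning noise, with the shapes added back and sphere masks shaping the falloff.
# Power-warped UVs distort the scrolling along the trail.
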Material /BRCosmetics/Weapons/FORT_Melee/Pickaxe_Sleepy_Time/FX/M_Pickaxe_SleepyTime_Trail
graph {
 properties {
  bUsedWithBeamTrails=true
  blendMode=BLEND_Translucent
  pixelDepthOffsetMode=PDOM_Legacy
  shadingModel=MSM_Unlit
  twoSided=true
  usedShadingModels=Unlit
 }
 nodes [
  AppendVector(A=Constant() {r=0.1}, B=Constant() {r=2.2})
  n1 = TextureSample(UVs=MaterialFunctionCall(XY=TextureCoordinate() {uTiling=0.75, vTiling=0.75}) {materialFunction=&Swizzle}.YX) {samplerType=SAMPLERTYPE_Masks, texture=/Game/Characters/Player/Male/Medium/Bodies/M_MED_SleepyTime/Textures/T_MED_SleepTime_Body_Mask01}
  n2 = LinearInterpolate(
   A=Constant3Vector() {constant=<rgba 0 0.688 0.9583 0>}
   B=Constant3Vector() {constant=<rgba 1.4 1.347 0.0366 0>}
   Alpha=Add(A=@n1.R, B=@n1.R)
  )
  n3 = TextureCoordinate()
  n4 = AppendVector(A=Power(Base=ComponentMask(None=@n3) {r=true}) {constExponent=4}, B=ComponentMask(None=@n3) {g=true})
  n5 = LinearInterpolate(A=LinearInterpolate(A=Panner() {speedX=-0.25}, B=Panner(Coordinate=@n4) {speedX=-0.45}, Alpha=@n1.R), B=Panner() {speedX=-0.5}, Alpha=@n1.B)
  n6 = Constant() {r=0.25}
  n7 = Constant() {r=2.5}
  n8 = TextureSample(UVs=Multiply(
    A=MaterialFunctionCall(XY=@n5) {materialFunction=&Swizzle}.YX
    B=AppendVector(A=@n6, B=@n7)
   )) {samplerType=SAMPLERTYPE_Masks, texture=/Game/Characters/Player/Male/Medium/Bodies/M_MED_SleepyTime/Textures/T_MED_SleepTime_Body_Mask02}
  n9 = Round(None=@n8.R)
  n10 = Multiply(A=Multiply(A=Multiply(A=@n2, B=@n9), B=@n9)) {constB=1.4}
  n11 = Constant3Vector() {constant=<rgba 0.8958 0.0252 0.2786 1>}
  n12 = Constant3Vector() {constant=<rgba 0.8906 0.0278 0.2559 1>}
  n13 = ComponentMask(None=TextureCoordinate()) {r=true}
  n14 = TextureCoordinate()
  n15 = AppendVector(A=Power(Base=ComponentMask(None=@n14) {r=true}) {constExponent=0.5}, B=ComponentMask(None=@n14) {g=true})
  n16 = AppendVector(A=Constant() {r=-1}, B=Constant())
  n17 = Constant() {r=0.5}
  n18 = Constant() {r=0.55}
  n19 = TextureSample(UVs=Multiply(A=Panner(Coordinate=@n15, Speed=@n16), B=AppendVector(A=@n17, B=@n18))) {samplerType=SAMPLERTYPE_Grayscale, texture=&LowResBlurredNoise}
  n20 = LinearInterpolate(Alpha=Multiply(A=Multiply(A=@n13, B=@n19.R)) {constB=3}) {constB=1.8}
  n21 = Clamp(None=Subtract(A=LinearInterpolate(A=@n11, B=@n12, Alpha=@n20), B=Multiply(A=@n9) {constB=2}))
  MaterialFunctionCall(Emissive=Desaturation(None=Add(A=@n10, B=@n21), Fraction=Constant() {r=-1})) {materialFunction=&MF_FX_EmissiveAdjustments}
  n22 = Constant() {r=1}
  n23 = TextureCoordinate()
  n24 = ComponentMask(None=@n23) {r=true}
  n25 = Constant() {r=6}
  n26 = Divide(A=Ceil(None=Multiply(A=@n24, B=@n25) {constB=8}), B=@n25) {constB=8}
  n27 = Divide(A=Ceil(None=TextureCoordinate() {uTiling=0, vTiling=6}), B=@n25) {constB=8}
  n28 = TextureSampleParameter2D!(
   UVs=Panner(Coordinate=MaterialFunctionCall(XY=@n27) {materialFunction=&Swizzle}.YX) {speedX=0.15, speedY=0.1}
   Level
   Apply View MipBias
  ) {
   group=Beat Texture
   mipValueMode=TMVM_MipLevel
   parameterName=Beat Texture Input
   sortPriority=-5
   texture=/Game/Effects/Fort_Effects/BluGlow/ZoomAnimation
  }
  n29 = Divide(A=Ceil(None=Multiply(A=@n28.R) {constB=16})) {constB=16}
  n30 = ComponentMask(None=@n23) {r=true}
  n31 = Add(A=Multiply(A=Multiply(A=@n26, B=@n29), B=Multiply(A=@n30, B=@n30)), B=Constant() {r=0.95})
  n32 = OneMinus(None=ComponentMask(None=TextureCoordinate()) {r=true})
  n33 = TextureCoordinate()
  n34 = Power(Base=ComponentMask(None=@n33) {r=true}) {constExponent=0.5}
  n35 = ComponentMask(None=@n33) {g=true}
  n36 = Constant() {r=-1}
  n37 = Constant()
  n38 = Multiply(
   A=Panner(Coordinate=AppendVector(A=@n34, B=@n35), Speed=AppendVector(A=@n36, B=@n37))
   B=AppendVector(A=Constant() {r=0.5}, B=Constant() {r=0.6})
  )
  n39 = Multiply(A=Multiply(
    A=Power(Base=@n32) {constExponent=5.2}
    B=TextureSample(UVs=@n38) {samplerType=SAMPLERTYPE_Grayscale, texture=&LowResBlurredNoise}.R
   )) {constB=5}
  n40 = Add(A=Ceil(None=Add(A=@n31) {constB=-1}), B=Round(None=Clamp(None=@n39)))
  n41 = SphereMask(A=Power(Base=@n13) {constExponent=0.5}, B=Constant() {r=0.5}) {attenuationRadius=0.5, hardnessPercent=0}
  n42 = TextureCoordinate() {uTiling=0.12}
  n43 = Power(Base=ComponentMask(None=@n42) {r=true}) {constExponent=0.8}
  n44 = ComponentMask(None=@n42) {g=true}
  Multiply(
   A=Add(A=Subtract(A=@n22, B=@n40), B=Multiply(A=@n9, B=@n41))
   B=SphereMask(A=AppendVector(A=@n43, B=@n44), B=Constant2Vector() {g=0.5, r=1}, Radius=Constant() {r=1}, Hardness=Constant() {r=0.88})
  )
 ]
}
