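# Unlit translucent two-sided mesh-particle material for line effects on the Sleepy Time pickaxe, driven per particle by dynamic parameters.
# SplineThicken sets line width from one parameter, while another, rounded up, selects pink, cyan or gold for the emissive adjustments function.
# A third offsets a clamped body mask texture vertically through a fixed-time Panner, so motion is particle-controlled.
# The mask's red channel supplies opacity, with depth fade skipped on mobile.
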Material /BRCosmetics/Weapons/FORT_Melee/Pickaxe_Sleepy_Time/FX/M_Pickaxe_SleepyTime_line
graph {
 properties {
  bUsedWithMeshParticles=true
  blendMode=BLEND_Translucent
  pixelDepthOffsetMode=PDOM_Legacy
  shadingModel=MSM_Unlit
  twoSided=true
  usedShadingModels=Unlit
 }
 nodes [
  n1 = DynamicParameter()
  n2 = Multiply(A=Constant() {r=2.6}, B=@n1.Param3)
  MaterialFunctionCall(WidthBase=@n2, WidthTip=@n2) {materialFunction=&SplineThicken}
  n3 = DynamicParameter()
  MaterialFunctionCall(Emissive=If(
    A=Constant() {r=2}
    B=Ceil(None=@n3.Param1)
    "A > B"=Constant3Vector() {constant=<rgba 1 0 0.3716 0>}
    "A == B"=Constant3Vector() {constant=<rgba 0.0156 0.8159 1 0>}
    "A < B"=Constant3Vector() {constant=<rgba 1 0.75 0 0>}
   )) {materialFunction=&MF_FX_EmissiveAdjustments}
  n4 = Constant()
  n5 = TextureSample(UVs=Panner(Coordinate=TextureCoordinate() {uTiling=0.09, vTiling=0.7}, Time=Constant() {r=1}, Speed=AppendVector(A=@n4, B=@n1.Param2)) {speedY=0.5}) {
   samplerSource=SSM_Clamp_WorldGroupSettings
   samplerType=SAMPLERTYPE_Masks
   texture=/Game/Characters/Player/Male/Medium/Bodies/M_MED_SleepyTime/Textures/T_MED_SleepTime_Body_Mask02
  }
  n6 = DepthFade(Opacity=@n5.R)
  ShadingPathSwitch(Default=@n6, Deferred=@n6, Mobile=@n5.R) {desc=Migrated from FeatureLevelSwitch}
 ]
}
