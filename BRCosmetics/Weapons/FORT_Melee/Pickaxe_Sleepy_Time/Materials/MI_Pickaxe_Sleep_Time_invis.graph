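# Translucent unlit material for the Sleepy Time pickaxe, named as its invisible variant and enabled for skeletal meshes and mesh deformers.
# A BaseColor tint is scaled by a parameterized Fresnel for a rim glow.
# A fake specular highlight dots the sun/moon vector from FortniteMaterialParameters with the reflection vector, then applies power and intensity.
# Both terms are summed into the Emissive input of MF_FX_EmissiveAdjustments.
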Material /BRCosmetics/Weapons/FORT_Melee/Pickaxe_Sleepy_Time/Materials/MI_Pickaxe_Sleep_Time_invis
graph {
 properties {
  bUsedWithMeshDeformer=true
  bUsedWithSkeletalMesh=true
  blendMode=BLEND_Translucent
  pixelDepthOffsetMode=PDOM_Legacy
  shadingModel=MSM_Unlit
  usedShadingModels=Unlit
 }
 nodes [
  n1 = VectorParameter() {parameterName=BaseColor, defaultValue=<rgba 0.8819 0.0567 0.1837 1>}
  n2 = Fresnel(
   ExponentIn=ScalarParameter() {defaultValue=1, parameterName=Expon}
   BaseReflectFractionIn=ScalarParameter() {defaultValue=0.02, parameterName=baseRef}
  )
  n3 = CollectionParameter() {
   collection=&FortniteMaterialParameters
   parameterName=SunAndMoonModelDirectionalVector
  }
  n4 = ComponentMask(None=@n3) {b=true, g=true, r=true}
  n5 = ReflectionVectorWS()
  n6 = Power(Base=Clamp(None=DotProduct(A=@n4, B=@n5)), Exp=ScalarParameter() {defaultValue=100, parameterName=Spec Highlight Power}) {constExponent=20}
  n7 = ScalarParameter() {defaultValue=5, parameterName=Spec Intense}
  MaterialFunctionCall(Emissive=Add(A=Multiply(A=@n1.RGB, B=@n2), B=Multiply(A=@n6, B=@n7))) {materialFunction=&MF_FX_EmissiveAdjustments}
 ]
}
