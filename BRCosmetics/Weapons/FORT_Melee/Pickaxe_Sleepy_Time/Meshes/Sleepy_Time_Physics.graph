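# Physics asset for the Sleepy Time pickaxe cosmetic, giving collision bodies and a joint to two of its dyn-prefixed bones.
# A sphere body sits offset on dyn_start, and a small capsule body covers dyn_main_shard.
# A single constraint joins the two bodies with both swing axes and twist set free, so the listed angular limit angles do not restrict rotation.
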PhysicsAsset /BRCosmetics/Weapons/FORT_Melee/Pickaxe_Sleepy_Time/Meshes/Sleepy_Time_Physics
bodies {dyn_main_shard [{shapeType=Capsule, length=2, radius=2}], dyn_start [{shapeType=Sphere, radius=5, center=<-2 -1 -2>}]}
constraints [{
  bone1Name=dyn_start
  bone2Name=dyn_main_shard
  swing1LimitDegrees=45
  swing1Motion=Free
  swing2LimitDegrees=45
  swing2Motion=Free
  twistLimitDegrees=10
  twistMotion=Free
 }]
