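# Anim Blueprint for the Sleepy Time pickaxe that makes its shards float and drift.
# Each update, game-time sine and cosine waves set additive bone-space offsets for the main shard and two fragments.
# These translations feed a RigidBody node with overridden gravity before the output pose.
# Swing-detection spin logic using socket distance, the dyn_start rotation and a trail controller are present but unconnected.
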AnimBlueprint /BRCosmetics/Weapons/FORT_Melee/Pickaxe_Sleepy_Time/Meshes/Sleepy_Time_AnimBP
target_skeleton=&/Sleepy_Time_Skeleton
graphs [
 AnimGraph {nodes [
   "Transform (Modify) Bone - Bone: dyn_start"(Alpha=1, Translation="0,0,0", Rotation=Get SpinConst() -> (SpinConst:Rotator), Scale="1,1,1") {rotationMode=BMM_Additive, rotationSpace=BCS_BoneSpace, alphaBoolBlend {blendOption=Linear}, boneToModify {boneName=dyn_start}}
   AnimGraphNode_SubInput_0 = Input Pose()
   n1 = "Transform (Modify) Bone - Bone: dyn_main_shard"(
    ComponentPose=Local To Component(LocalPose=@AnimGraphNode_SubInput_0)
    Alpha=1
    Translation=Get ShardFloatB() -> (ShardFloatB:Vector)
    Rotation="0,0,0"
    Scale="1,1,1"
   ) {translationMode=BMM_Additive, translationSpace=BCS_BoneSpace, alphaBoolBlend {blendOption=Linear}, boneToModify {boneName=dyn_main_shard}}
   n2 = "Transform (Modify) Bone - Bone: dyn_fragment_01"(ComponentPose=@n1, Alpha=1, Translation=Get ShardFloatA() -> (ShardFloatA:Vector), Rotation="0,0,0", Scale="1,1,1") {translationMode=BMM_Additive, translationSpace=BCS_BoneSpace, alphaBoolBlend {blendOption=Linear}, boneToModify {boneName=dyn_fragment_01}}
   n3 = "Transform (Modify) Bone - Bone: dyn_fragment_02"(ComponentPose=@n2, Alpha=1, Translation=Get ShardFloatC() -> (ShardFloatC:Vector), Rotation="0,0,0", Scale="1,1,1") {translationMode=BMM_Additive, translationSpace=BCS_BoneSpace, alphaBoolBlend {blendOption=Linear}, boneToModify {boneName=dyn_fragment_02}}
   n4 = RigidBody(ComponentPose=@n3) {
    bClampLinearTranslationLimitToRefPose=true
    bOverrideWorldGravity=true
    overridePhysicsAsset=&/Sleepy_Time_Physics
    alphaBoolBlend {blendOption=Linear}
    componentAppliedLinearAccClamp=<3000 3000 3000>
    componentLinearAccScale=<1 1 1>
    componentLinearVelScale=<0.3 0.3 0.3>
    overrideWorldGravity {z=-100}
   }
   Output Pose(Result=Component To Local(ComponentPose=@n4)) {layerGroup=None}
   Trail controller - Bone: dyn_swingMeasure(Alpha=1) {
    rotationLimits [{limitMax=<180 180 180>}, {limitMax=<180 180 180>}]
    rotationOffsets [<0 0 0>, <0 0 0>]
    trailBone {boneName=dyn_swingMeasure}
    bLimitStretch=~
   }
  ]}
 EventGraph {nodes [
   Event BlueprintUpdateAnimation()
   n1 = GetGameTimeInSeconds()
   n2 = float * float(A=float * float(A=1, B=0.5), B=@n1)
   n3 = float * float(A=float * float(A=1, B=1.25), B=@n1)
   n4 = MakeVector(X=float * float(A="Sin (Radians)"(A=@n3), B=0.8), Y=float * float(A="Sin (Radians)"(A=@n2), B=0.3))
   > Set ShardFloatA(ShardFloatA:Vector=@n4) -> (Output_Get:Vector)
   > Set ShardFloatC(ShardFloatC:Vector=vector * float(A=@n4, B=-0.5)) -> (Output_Get:Vector)
   n5 = "Cos (Radians)"!(A:double=float * float(A=float * float(A=1, B=1), B=@n1)) -> (ReturnValue:double)
   n6 = "Cos (Radians)"(A=float * float(A=float * float(A=1, B=1.25), B=@n1))
   > Set ShardFloatB(ShardFloatB:Vector=MakeVector(Y=float * float(A=@n5, B=0.2), Z=float * float(A=@n6, B=1))) -> (Output_Get:Vector)
   n7 = Sequence!(execute) -> (then_0, then_1, then_2:Exec)
   n8 = GetPartSkeletalMeshComponent()
   > Set SwingBlend(SwingBlend:bool="float >= float"(
     A=Get Distance Between Two Sockets!(
      Component:Skeletal Mesh Component Object Reference=@n8
      SocketOrBoneNameA:Name=axeTop
      SocketSpaceA:ERelativeTransformSpace Enum=RTS_Actor
      SocketOrBoneNameB:Name=axeSwing
      SocketSpaceB:ERelativeTransformSpace Enum=RTS_Actor
      bRemapRange:bool=false
      InRangeMin:float=4
      InRangeMax:float=45
      OutRangeMin:float=0
      OutRangeMax:float=1
     ) -> (ReturnValue:float) {is=title K2Node_CallFunction}
     B=40
    ) {is=title K2Node_CallFunction}) -> (Output_Get:bool)
   n9 = Branch(execute=@n7.then_1, Condition=Get SwingBlend() -> (SwingBlend:bool))
   > Set SwingDelta(SwingDelta:double=FInterpTo(Current=Get SwingDelta() -> (SwingDelta:double), Target=10, DeltaTime=0, InterpSpeed=1)) -> (Output_Get:double)
   > n10 = Branch(Condition="float >= float"(A=Get SpinX() -> (SpinX:double), B=720))
   > n11 = Set SpinX(SpinX:double=float - float(A=Get SpinX(), B=720) {is=title K2Node_CallFunction}) -> (Output_Get:double)
   n12 = Get SpinX()
   Set SpinConst(execute=[@n14.then, @n11.then], SpinConst:Rotator=MakeRotator(Roll=@n12, Pitch=@n12)) -> (Output_Get:Rotator)
   n13 = Get SwingDelta()
   n14 = Set SpinX(
    execute=@n10.else
    SpinX=float + float(A=Get SpinX(), B=float * float(A=@n13, B=0.75)) {is=title K2Node_CommutativeAssociativeBinaryOperator}
   )
   Set SwingDelta(execute=@n9.else, SwingDelta=FInterpTo(Current=Get SwingDelta(), Target=1, DeltaTime=0, InterpSpeed=1))
   > n15 = Branch(Condition="float <= float"!(A:double=Get SpinX(), B:double=0) -> (ReturnValue:bool) {is=title K2Node_CallFunction})
   > n16 = Set SpinX(SpinX=0)
   n17 = Get SpinX()
   Set SpinConst(execute=[@n16.then, @n19.then], SpinConst=MakeRotator(Roll=@n17, Pitch=@n17))
   n18 = Get SwingDelta()
   n19 = Set SpinX(execute=@n15.else, SpinX=float - float(A=Get SpinX(), B=float * float(A=@n18, B=0.75)))
   n20 = Branch(Condition=Get SwingBlend())
   > Set SwingDelta(SwingDelta=FInterpTo(Current=Get SwingDelta(), Target=10, DeltaTime=0, InterpSpeed=1))
   Set SwingDelta(execute=@n20.else, SwingDelta=FInterpTo(Current=Get SwingDelta(), Target=1, DeltaTime=0, InterpSpeed=1))
   TryGetPawnOwner()
   n21 = float * float(A=1, B=2)
   float * float(A="Sin (Radians)"(A=float * float(A=@n21, B=@n1)), B=0)
   n22 = float * float(A=1, B=0.5)
   float * float(A="Cos (Radians)"(A=float * float(A=@n22, B=@n1)), B=0)
  ]}
]
variables [
 ShardFloatA {type=Vector}
 ShardFloatB {type=Vector}
 ShardFloatC {type=Vector}
 SpinConst {type=Rotator}
 SpinX {type=number, default=1}
 SwingBlend {type=boolean}
 SwingDelta {type=number}
]
